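# Niagara FX material function that adds a texture parameter, an inky smoke trail by default, into base color as a foundation layer.
# It samples the texture at the input UVs (UV0 by default) and adds its RGB to the incoming attributes' base color and an optional base color value input.
# The sum is written back into Material Attributes and also output as a raw value, supporting both attribute and traditional material workflows.
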MaterialFunction /Game/Effects/Niagara/MasterMaterials/Functions/V2/BaseColor/MF_FX_BaseColorTexture_v02
graph {
 properties {bExposeToLibrary=true, description=Material Attributes with edited base color., libraryCategoriesText ["FX 1) Base Color"]}
 nodes [
  n1 = FunctionInput() {
   bUsePreviewValueAsDefault=true
   desc=Previous material attributes in
   description="If you are using this, you shouldn't need to use Input Base Color Value."
   inputType=FunctionInput_MaterialAttributes
  }
  n2 = GetMaterialAttributes(None=@n1) {desc=crack open material attributes to add in this functions modifications, attributeGetTypes [#1]}
  n3 = FunctionInput(Preview=TextureCoordinate() {desc=Default}) {
   bUsePreviewValueAsDefault=true
   desc=UV input
   description=UV input for texture. Default is UV0.
   inputName=Uvs
   inputType=FunctionInput_Vector2
   sortPriority=1
  }
  n4 = TextureSampleParameter2D(UVs=@n3) {
   desc=This texture provides the foundation for the other channels like opacity and emissive color.
   group="1.0) Base Color Texture"
   parameterName=Texture
   texture=/Game/Effects/Fort_Effects/Textures/T_InkySmokeTrail
  }
  n5 = FunctionInput() {
   bUsePreviewValueAsDefault=true
   desc=You shouldnt use this if you are using material attributes. This is for a traditional material workflow.
   description=You shouldnt use this if you are using material attributes. This is for a traditional material workflow.
   inputName=Base Color Value
   sortPriority=2
  }
  n6 = Add(A=Add(A=@n2.BaseColor, B=@n4.RGB) {desc="add into a previous material attribute, maybe add a switch for different blending modes."}, B=@n5) {desc=add in a previous base color value}
  n7 = SetMaterialAttributes!(MaterialAttributes=@n2.MaterialAttributes, BaseColor=@n6) {
   desc="set base color attribute. Most likely it will be edited down the chain in \r\nanother material function."
   attributeSetTypes [#1]
  }
  FunctionOutput(None=@n7) {desc="Material Attributes with edited base color.\r\n", outputName=MA BaseColor}
  FunctionOutput(None=@n6) {
   desc=Vector 3 of texture combined with Input Base Color Value or the input of material attributes base color. You should use one or the other.
   outputName=Base Color Value
   sortPriority=1
  }
 ]
}
